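# Physics asset for the female medium FierceBraid player body, covering dynamic skirt bones with capsule bodies on pelvis and both thighs.
# Seven skirt strands (front, back, paired middle and back-middle, left-only side) each chain base, first and second capsules through angle-limited constraints.
# An extra, tightly swing-limited constraint ties the front tip to the left middle tip, linking those strands.
# No constraints join the strand bases, pelvis or thighs to one another.
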PhysicsAsset /Game/Characters/Player/Female/Medium/Bodies/F_MED_FierceBraid/Meshes/F_MED_FierceBraid_Physics
bodies {
 dyn_skirt_back_1 [{shapeType=Capsule, length=5, radius=1.5, shapeName=dyn_skirt_back_1_capsule, center=<4.4822 0 0>, rotation=<rot 90 -160.5288 -160.5286>}]
 dyn_skirt_back_2 [{shapeType=Capsule, length=4.5, radius=4, shapeName=dyn_skirt_back_2_capsule, center=<5.4078 0.5109 0>, rotation=<rot 90 -90 -89.9998>}]
 dyn_skirt_back_base [{shapeType=Capsule, length=0, radius=2.5, shapeName=dyn_skirt_back_base_capsule}]
 dyn_skirt_back_middle_1_l [{
   shapeType=Capsule
   length=5
   radius=1.5
   shapeName=dyn_skirt_back_middle_1_l_capsule
   center=<4.4822 0 0>
   rotation=<rot 90 -179.8734 -179.8733>
  }]
 dyn_skirt_back_middle_1_r [{shapeType=Capsule, length=5, radius=1.5, shapeName=dyn_skirt_back_middle_1_r_capsule, center=<4.4822 0 0>, rotation=<rot 90 90 90.0002>}]
 dyn_skirt_back_middle_2_l [{shapeType=Capsule, length=4.5, radius=4, shapeName=dyn_skirt_back_middle_2_l_capsule, center=<5.4078 0.5109 0>, rotation=<rot 90 0 0.0002>}]
 dyn_skirt_back_middle_2_r [{
   shapeType=Capsule
   length=4.5
   radius=4
   shapeName=dyn_skirt_back_middle_2_r_capsule
   center=<5.4078 0.5109 0>
   rotation=<rot 90 174.9488 174.949>
  }]
 dyn_skirt_back_middle_base_l [{shapeType=Capsule, length=0, radius=2.5, shapeName=dyn_skirt_back_middle_base_l_capsule}]
 dyn_skirt_back_middle_base_r [{shapeType=Capsule, length=0, radius=2.5, shapeName=dyn_skirt_back_middle_base_r_capsule}]
 dyn_skirt_front_1 [{shapeType=Capsule, length=6.5, radius=5, shapeName=dyn_skirt_front_1_capsule, center=<7.4501 0 0>, rotation=<rot 0 0.0004 0>}]
 dyn_skirt_front_2 [{shapeType=Capsule, length=7, radius=3.5, shapeName=dyn_skirt_front_2_capsule, center=<7.3985 1.4144 1.5104>, rotation=<rot 25 0.0002 -5>}]
 dyn_skirt_front_base [{shapeType=Capsule, length=0, radius=2.5, shapeName=dyn_skirt_front_base_capsule}]
 dyn_skirt_middle_1_l [{shapeType=Capsule, length=6.5, radius=2, shapeName=dyn_skirt_middle_1_l_capsule, center=<4.4822 0 0>, rotation=<rot 90 180 179.9998>}]
 dyn_skirt_middle_1_r [{shapeType=Capsule, length=5, radius=1.5, shapeName=dyn_skirt_middle_1_r_capsule, center=<4.4822 0 0>, rotation=<rot 90 180 -180>}]
 dyn_skirt_middle_2_l [{
   shapeType=Capsule
   length=6.5
   radius=3
   shapeName=dyn_skirt_middle_2_l_capsule
   center=<3.7035 1.1109 2.3837>
   rotation=<rot 9.9999 0.0009 -10.0001>
  }]
 dyn_skirt_middle_2_r [{
   shapeType=Capsule
   length=6
   radius=2.5
   shapeName=dyn_skirt_middle_2_r_capsule
   center=<5.4078 0.5109 0>
   rotation=<rot 90 179.9802 179.9802>
  }]
 dyn_skirt_middle_base_l [{shapeType=Capsule, length=0, radius=2.5, shapeName=dyn_skirt_middle_base_l_capsule}]
 dyn_skirt_middle_base_r [{shapeType=Capsule, length=0, radius=2.5, shapeName=dyn_skirt_middle_base_r_capsule}]
 dyn_skirt_side_1_l [{shapeType=Capsule, length=4.5, radius=2, shapeName=dyn_skirt_side_1_l_capsule, center=<4.4822 0 0>, rotation=<rot 90 -0.0395 -0.0396>}]
 dyn_skirt_side_2_l [{
   shapeType=Capsule
   length=5.5
   radius=2.5
   shapeName=dyn_skirt_side_2_l_capsule
   center=<5.0698 1.0919 -0.7336>
   rotation=<rot -0.0005 -0.0002 9.9999>
  }]
 dyn_skirt_side_base_l [{shapeType=Capsule, length=0, radius=2.5, shapeName=dyn_skirt_side_base_l_capsule}]
 pelvis [{shapeType=Capsule, length=9, radius=10, shapeName=pelvis_capsule, center=<-4.362 -1.3892 0>}]
 thigh_l [{shapeType=Capsule, length=29, radius=7.5, shapeName=thigh_l_capsule, center=<-21.6867 -0.2939 -0.0001>, rotation=<rot 90 160.5288 160.5291>}]
 thigh_r [{shapeType=Capsule, length=28, radius=7, shapeName=thigh_r_capsule, center=<21.6184 0.2939 0.0001>, rotation=<rot 90 90 90.0004>}]
}
constraints [
 {bone1Name=dyn_skirt_side_1_l, bone2Name=dyn_skirt_side_base_l, swing1LimitDegrees=10, swing2LimitDegrees=5, twistLimitDegrees=5}
 {bone1Name=dyn_skirt_middle_1_r, bone2Name=dyn_skirt_middle_base_r, swing1LimitDegrees=10, swing2LimitDegrees=5, twistLimitDegrees=5}
 {bone1Name=dyn_skirt_middle_1_l, bone2Name=dyn_skirt_middle_base_l, swing1LimitDegrees=45, swing2LimitDegrees=5, twistLimitDegrees=5}
 {bone1Name=dyn_skirt_front_1, bone2Name=dyn_skirt_front_base, swing1LimitDegrees=15, swing2LimitDegrees=45, twistLimitDegrees=1}
 {
  bone1Name=dyn_skirt_back_middle_1_r
  bone2Name=dyn_skirt_back_middle_base_r
  swing1LimitDegrees=10
  swing2LimitDegrees=5
  twistLimitDegrees=5
 }
 {
  bone1Name=dyn_skirt_back_middle_1_l
  bone2Name=dyn_skirt_back_middle_base_l
  swing1LimitDegrees=10
  swing2LimitDegrees=5
  twistLimitDegrees=5
 }
 {bone1Name=dyn_skirt_back_1, bone2Name=dyn_skirt_back_base, swing1LimitDegrees=10, swing2LimitDegrees=5, twistLimitDegrees=5}
 {bone1Name=dyn_skirt_side_2_l, bone2Name=dyn_skirt_side_1_l, swing1LimitDegrees=35, swing2LimitDegrees=1, twistLimitDegrees=1}
 {bone1Name=dyn_skirt_middle_2_r, bone2Name=dyn_skirt_middle_1_r, swing1LimitDegrees=35, swing2LimitDegrees=10, twistLimitDegrees=20}
 {bone1Name=dyn_skirt_middle_2_l, bone2Name=dyn_skirt_middle_1_l, swing1LimitDegrees=35, swing2LimitDegrees=10, twistLimitDegrees=10}
 {bone1Name=dyn_skirt_back_middle_2_r, bone2Name=dyn_skirt_back_middle_1_r, swing1LimitDegrees=35, swing2LimitDegrees=10, twistLimitDegrees=10}
 {bone1Name=dyn_skirt_back_middle_2_l, bone2Name=dyn_skirt_back_middle_1_l, swing1LimitDegrees=35, swing2LimitDegrees=10, twistLimitDegrees=10}
 {bone1Name=dyn_skirt_back_2, bone2Name=dyn_skirt_back_1, swing1LimitDegrees=35, swing2LimitDegrees=10, twistLimitDegrees=10}
 {bone1Name=dyn_skirt_front_2, bone2Name=dyn_skirt_middle_2_l, swing1LimitDegrees=1, swing2LimitDegrees=1, twistLimitDegrees=20}
 {bone1Name=dyn_skirt_front_2, bone2Name=dyn_skirt_front_1, swing1LimitDegrees=20, swing2LimitDegrees=30, twistLimitDegrees=45}
]
