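# Post-process style AnimBlueprint for the female FierceBraid body that copies the attached parent's pose and adjusts skirt, dragon, sleeve and bow bones.
# Additive bone offsets feed RigidBody physics, then slide-specific skirt offsets, AnimDynamics on the hand bow and a deform Control Rig.
# The update event sets these offsets and blend alphas from sliding, crouching, parachute and skydiving states.
# A chain of skirt rotation constraints is left unconnected to the output.
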AnimBlueprint /Game/Characters/Player/Female/Medium/Bodies/F_MED_FierceBraid/Meshes/F_MED_FierceBraid_AnimBP
target_skeleton=&/F_MED_FierceBraid_Skeleton
graphs [
 AnimGraph {nodes [
   n1 = "Constraint - dyn_skirt_middle_base_l (RR)"!(ComponentPose, Alpha=1, ConstraintWeights_0=0.9, ConstraintWeights_1:float=0.1) -> (Pose) {
    alphaBoolBlend {blendOption=Linear}
    alphaScaleBiasClamp {bInterpResult=true, interpSpeedDecreasing=100, interpSpeedIncreasing=100}
    boneToModify {boneName=dyn_skirt_middle_base_l}
    constraintSetup [{transformType=Rotation, targetBone {boneName=thigh_l}}, {transformType=Rotation, targetBone {boneName=pelvis}}]
    constraintWeights [1,1]
   }
   n2 = "Constraint - dyn_skirt_side_base_l (RR)"!(ComponentPose=@n1, Alpha=1, ConstraintWeights_0=0.1, ConstraintWeights_1:float=0.9) -> (Pose) {
    alphaBoolBlend {blendOption=Linear}
    alphaScaleBiasClamp {bInterpResult=true, interpSpeedDecreasing=100, interpSpeedIncreasing=100}
    boneToModify {boneName=dyn_skirt_side_base_l}
    constraintSetup [{transformType=Rotation, targetBone {boneName=pelvis}}, {transformType=Rotation, targetBone {boneName=thigh_l}}]
    constraintWeights [1,1]
   }
   n3 = "Constraint - dyn_skirt_back_middle_base_l (RR)"!(ComponentPose=@n2, Alpha=1, ConstraintWeights_0=0.75, ConstraintWeights_1:float=0.25) -> (Pose) {
    alphaBoolBlend {blendOption=Linear}
    alphaScaleBiasClamp {bInterpResult=true, interpSpeedDecreasing=100, interpSpeedIncreasing=100}
    boneToModify {boneName=dyn_skirt_back_middle_base_l}
    constraintSetup [{transformType=Rotation, targetBone {boneName=pelvis}}, {transformType=Rotation, targetBone {boneName=thigh_l}}]
    constraintWeights [1,1]
   }
   n4 = "Constraint - dyn_skirt_middle_base_r (RR)"!(ComponentPose=@n3, Alpha=0.5, ConstraintWeights_0=0.5, ConstraintWeights_1:float=0.5) -> (Pose) {
    alphaBoolBlend {blendOption=Linear}
    alphaScaleBiasClamp {bInterpResult=true, interpSpeedDecreasing=100, interpSpeedIncreasing=100}
    boneToModify {boneName=dyn_skirt_middle_base_r}
    constraintSetup [{transformType=Rotation, targetBone {boneName=pelvis}}, {transformType=Rotation, targetBone {boneName=thigh_r}}]
    constraintWeights [1,1]
   }
   "Constraint - dyn_skirt_back_middle_base_r (RR)"!(ComponentPose=@n4, Alpha=0.5, ConstraintWeights_0=0.8, ConstraintWeights_1:float=0.2) -> (Pose) {
    alphaBoolBlend {blendOption=Linear}
    alphaScaleBiasClamp {bInterpResult=true, interpSpeedDecreasing=100, interpSpeedIncreasing=100}
    boneToModify {boneName=dyn_skirt_back_middle_base_r}
    constraintSetup [{transformType=Rotation, targetBone {boneName=dyn_skirt_back_1}}, {transformType=Rotation, targetBone {boneName=dyn_skirt_middle_1_r}}]
    constraintWeights [1,1]
   }
   n5 = "Transform (Modify) Bone - Bone: dyn_skirt_front_base"(
    ComponentPose=Local To Component(LocalPose=Copy Pose From Mesh() {bUseAttachedParent=true})
    Alpha=1
    Translation=Get Dyn_front_skirt_translate() -> (Dyn_front_skirt_translate:Vector)
    Rotation=Get Dyn_front_skirt_rotate() -> (Dyn_front_skirt_rotate:Rotator)
    Scale="1,1,1"
   ) {
    rotationMode=BMM_Additive
    rotationSpace=BCS_BoneSpace
    translationMode=BMM_Additive
    translationSpace=BCS_BoneSpace
    alphaBoolBlend {blendOption=Linear}
    alphaScaleBiasClamp {bInterpResult=true, interpSpeedDecreasing=100, interpSpeedIncreasing=100}
    boneToModify {boneName=dyn_skirt_front_base}
    translation {y=10}
   }
   n6 = "Transform (Modify) Bone - Bone: dyn_skirt_middle_base_l"(
    ComponentPose=@n5
    Alpha=1
    Translation=Get Dyn_middle_skirt_translate() -> (Dyn_middle_skirt_translate:Vector)
    Rotation=Get Dyn_middle_skirt_rotate_l() -> (Dyn_middle_skirt_rotate_l:Rotator)
    Scale="1,1,1"
   ) {
    rotationMode=BMM_Additive
    rotationSpace=BCS_BoneSpace
    translationMode=BMM_Additive
    translationSpace=BCS_BoneSpace
    alphaBoolBlend {blendOption=Linear}
    alphaScaleBiasClamp {bClampResult=true, bInterpResult=true, interpSpeedDecreasing=100, interpSpeedIncreasing=100}
    boneToModify {boneName=dyn_skirt_middle_base_l}
   }
   n7 = "Transform (Modify) Bone - Bone: dyn_skirt_side_base_l"(ComponentPose=@n6, Alpha=1, Translation="0,0,0", Rotation=Get Dyn_side_skirt_rotate_l() -> (Dyn_side_skirt_rotate_l:Rotator), Scale="1,1,1") {
    rotationMode=BMM_Additive
    rotationSpace=BCS_BoneSpace
    translationMode=BMM_Additive
    translationSpace=BCS_BoneSpace
    alphaBoolBlend {blendOption=Linear}
    alphaScaleBiasClamp {bClampResult=true, bInterpResult=true, interpSpeedDecreasing=100, interpSpeedIncreasing=100}
    boneToModify {boneName=dyn_skirt_side_base_l}
   }
   n8 = "Transform (Modify) Bone - Bone: dyn_skirt_back_middle_base_l"(
    ComponentPose=@n7
    Alpha=1
    Translation="0,0,0"
    Rotation=Get Dyn_back_middle_skirt_rotate_l() -> (Dyn_back_middle_skirt_rotate_l:Rotator)
    Scale="1,1,1"
   ) {
    rotationMode=BMM_Additive
    rotationSpace=BCS_BoneSpace
    translationMode=BMM_Additive
    translationSpace=BCS_BoneSpace
    alphaBoolBlend {blendOption=Linear}
    alphaScaleBiasClamp {bClampResult=true, bInterpResult=true, interpSpeedDecreasing=100, interpSpeedIncreasing=100}
    boneToModify {boneName=dyn_skirt_back_middle_base_l}
   }
   n9 = "Transform (Modify) Bone - Bone: dyn_skirt_back_base"(ComponentPose=@n8, Alpha=1, Translation="0,0,0", Rotation=Get Dyn_back_skirt_rotate() -> (Dyn_back_skirt_rotate:Rotator), Scale="1,1,1") {
    rotationMode=BMM_Additive
    rotationSpace=BCS_BoneSpace
    translationMode=BMM_Additive
    translationSpace=BCS_BoneSpace
    alphaBoolBlend {blendOption=Linear}
    alphaScaleBiasClamp {bClampResult=true, bInterpResult=true, interpSpeedDecreasing=100, interpSpeedIncreasing=100}
    boneToModify {boneName=dyn_skirt_back_base}
   }
   n10 = "Transform (Modify) Bone - Bone: dyn_dragon_head"(
    ComponentPose=@n9
    Alpha=1
    Translation=Get Dyn_Dragon_head_translate() -> (Dyn_Dragon_head_translate:Vector)
    Rotation=Get Dyn_Dragon_head_rotate() -> (Dyn_Dragon_head_rotate:Rotator)
    Scale="1,1,1"
   ) {
    rotationMode=BMM_Additive
    rotationSpace=BCS_BoneSpace
    translationMode=BMM_Additive
    translationSpace=BCS_BoneSpace
    alphaBoolBlend {blendOption=Linear}
    alphaScaleBiasClamp {bClampResult=true, bInterpResult=true, clampMin=0.9, interpSpeedDecreasing=1, interpSpeedIncreasing=1}
    boneToModify {boneName=dyn_dragon_head}
    rotation {pitch=-10, yaw=10}
   }
   n11 = "Transform (Modify) Bone - Bone: dyn_dragon_spine_1"(
    ComponentPose=@n10
    Alpha=1
    Translation=Get Dyn_Dragon_spine_translate() -> (Dyn_Dragon_spine_translate:Vector)
    Rotation="0,0,0"
    Scale="1,1,1"
   ) {
    rotationMode=BMM_Additive
    rotationSpace=BCS_BoneSpace
    translationMode=BMM_Additive
    translationSpace=BCS_BoneSpace
    alphaBoolBlend {blendOption=Linear}
    alphaScaleBiasClamp {bClampResult=true, bInterpResult=true, clampMin=0.9, interpSpeedDecreasing=1, interpSpeedIncreasing=1}
    boneToModify {boneName=dyn_dragon_spine_1}
   }
   n12 = Get Dyn_alfa_transfom() -> (Dyn_alfa_transfom:double)
   n13 = "Transform (Modify) Bone - Bone: deform_upperarm_bk_l"(ComponentPose=@n11, Alpha=@n12, Translation="0,-4.5,0", Rotation="0,0,0", Scale="1,1,1") {
    rotationMode=BMM_Additive
    rotationSpace=BCS_BoneSpace
    translationMode=BMM_Additive
    translationSpace=BCS_BoneSpace
    alphaBoolBlend {blendOption=Linear}
    alphaScaleBiasClamp {bClampResult=true, bInterpResult=true, interpSpeedDecreasing=1, interpSpeedIncreasing=25}
    boneToModify {boneName=deform_upperarm_bk_l}
    translation {y=-4.5}
   }
   n14 = "Transform (Modify) Bone - Bone: dyn_sleeve_down_l"(ComponentPose=@n13, Alpha=@n12, Translation="0,-4,2", Rotation="0,0,0", Scale="1,1,1") {
    rotationMode=BMM_Additive
    rotationSpace=BCS_BoneSpace
    translationMode=BMM_Additive
    translationSpace=BCS_BoneSpace
    alphaBoolBlend {blendOption=Linear}
    alphaScaleBiasClamp {bClampResult=true, bInterpResult=true, interpSpeedDecreasing=1, interpSpeedIncreasing=25}
    boneToModify {boneName=dyn_sleeve_down_l}
    translation {y=-4, z=2}
   }
   n15 = RigidBody(ComponentPose=@n14) {
    bOverrideWorldGravity=true
    overridePhysicsAsset=&/F_MED_FierceBraid_Physics
    simulationSpace=BaseBoneSpace
    alphaBoolBlend {blendOption=Linear}
    alphaScaleBiasClamp {bClampResult=true, bInterpResult=true}
    baseBoneRef {boneName=pelvis}
    componentAppliedLinearAccClamp=<300 300 300>
    componentLinearAccScale=<0.75 0.75 0.75>
    componentLinearVelScale=<0.75 0.75 0.75>
    overrideWorldGravity {z=-980}
    simSpaceSettings {maxAngularAcceleration=5, maxAngularVelocity=1, maxLinearAcceleration=4, maxLinearVelocity=100, velocityScaleZ=0, worldAlpha=0.25}
   }
   n16 = Get Dyn_alfa_slide() -> (Dyn_alfa_slide:double)
   n17 = "Transform (Modify) Bone - Bone: dyn_skirt_back_middle_2_r"(
    ComponentPose=@n15
    Alpha=@n16
    Translation="0,0,0"
    Rotation=Get Dyn_back_slide_skirt_rotate_r() -> (Dyn_back_slide_skirt_rotate_r:Rotator)
    Scale="1,1,1"
   ) {
    rotationMode=BMM_Additive
    rotationSpace=BCS_BoneSpace
    translationMode=BMM_Additive
    translationSpace=BCS_BoneSpace
    alphaBoolBlend {blendOption=Linear}
    alphaScaleBiasClamp {bClampResult=true, bInterpResult=true, interpSpeedDecreasing=50, interpSpeedIncreasing=1}
    boneToModify {boneName=dyn_skirt_back_middle_2_r}
    rotation=<rot -35 -120 -30>
   }
   n18 = "Transform (Modify) Bone - Bone: dyn_skirt_middle_2_r"(
    ComponentPose=@n17
    Alpha=@n16
    Translation="0,0,0"
    Rotation=Get Dyn_middle_slide_skirt_rotate_r() -> (Dyn_middle_slide_skirt_rotate_r:Rotator)
    Scale="1,1,1"
   ) {
    rotationMode=BMM_Additive
    rotationSpace=BCS_BoneSpace
    translationMode=BMM_Additive
    translationSpace=BCS_BoneSpace
    alphaBoolBlend {blendOption=Linear}
    alphaScaleBiasClamp {bClampResult=true, bInterpResult=true, interpSpeedDecreasing=50, interpSpeedIncreasing=1}
    boneToModify {boneName=dyn_skirt_middle_2_r}
    rotation=<rot 15 -90 -30>
   }
   n19 = "Transform (Modify) Bone - Bone: dyn_skirt_front_1"(
    ComponentPose=@n18
    Alpha=1
    Translation="0,0,0"
    Rotation=Get Dyn_front_skirt_1_rotate() -> (Dyn_front_skirt_1_rotate:Rotator)
    Scale="1,1,1"
   ) {
    rotationMode=BMM_Additive
    rotationSpace=BCS_BoneSpace
    translationMode=BMM_Additive
    translationSpace=BCS_BoneSpace
    alphaBoolBlend {blendOption=Linear}
    alphaScaleBiasClamp {bClampResult=true, bInterpResult=true, interpSpeedDecreasing=50, interpSpeedIncreasing=1}
    boneToModify {boneName=dyn_skirt_front_1}
    rotation {yaw=20}
   }
   n20 = "Transform (Modify) Bone - Bone: dyn_skirt_middle_1_l"(
    ComponentPose=@n19
    Alpha=@n16
    Translation="0,0,0"
    Rotation=Get Dyn_front_skirt_2_rotate() -> (Dyn_front_skirt_2_rotate:Rotator)
    Scale="1,1,1"
   ) {
    rotationMode=BMM_Additive
    rotationSpace=BCS_BoneSpace
    translationMode=BMM_Additive
    translationSpace=BCS_BoneSpace
    alphaBoolBlend {blendOption=Linear}
    alphaScaleBiasClamp {bClampResult=true, bInterpResult=true, interpSpeedDecreasing=50, interpSpeedIncreasing=1}
    boneToModify {boneName=dyn_skirt_middle_1_l}
    rotation {yaw=20}
   }
   n21 = Anim Dynamics - Bone: dyn_hand_bow(ComponentPose=@n20, Alpha=1, ExternalForce="0,0,0") {
    alpha=0.9
    angularSpringConstant=300
    bAngularSpring=true
    bUseGravityOverride=true
    numSolverIterationsPostUpdate=4
    numSolverIterationsPreUpdate=16
    simulationSpace=BoneRelative
    alphaBoolBlend {blendOption=Linear}
    boundBone {boneName=dyn_hand_bow}
    componentAppliedLinearAccClamp=<3000 3000 3000>
    componentLinearAccScale=<1 1 1>
    componentLinearVelScale=<1 1 1>
    physicsBodyDefinitions [{
      sphereCollisionRadius=0
      boundBone {boneName=dyn_hand_bow}
      boxExtents=<3 5 3>
      constraintSetup {angularTargetAxis=AxisZ, twistAxis=AxisY, angularLimitsMax=<5 15 30>, angularLimitsMin=<-5 -15 -30>, angularTarget=<1 0 0>}
      localJointOffset=<2 0 0>
     }]
    relativeSpaceBone {boneName=upperarm_r}
   }
   n22 = Control Rig(Source=Component To Local(ComponentPose=@n21)) {
    alphaBoolBlend {blendOption=Linear}
    controlRigAssetReference {blueprintRigClass=&/F_MED_FierceBraid_DeformRig.F_MED_FierceBraid_DeformRig_C}
   }
   Output Pose(Result=@n22) {layerGroup=None}
  ]}
 EventGraph {nodes [
   Event BlueprintUpdateAnimation()
   > n1 = Sequence()
   > n2 = Branch(Condition=Get bIsSliding())
   > Set Dyn_alfa_slide(Dyn_alfa_slide:double=1) -> (Output_Get:double)
   > n3 = Branch(Condition=Get bIsBaseballSliding())
   > Set Dyn_back_slide_skirt_rotate_r(Dyn_back_slide_skirt_rotate_r:Rotator="0,-175,0") -> (Output_Get:Rotator)
   > Set Dyn_middle_slide_skirt_rotate_r(Dyn_middle_slide_skirt_rotate_r:Rotator="0,-150,0") -> (Output_Get:Rotator)
   > Set Dyn_front_skirt_1_rotate(Dyn_front_skirt_1_rotate:Rotator="0,-90,0") -> (Output_Get:Rotator)
   > Set Dyn_front_skirt_2_rotate(Dyn_front_skirt_2_rotate:Rotator="-60,-90,0") -> (Output_Get:Rotator)
   Set Dyn_back_slide_skirt_rotate_r(execute=@n3.else, Dyn_back_slide_skirt_rotate_r="0,-90,0")
   > Set Dyn_middle_slide_skirt_rotate_r(Dyn_middle_slide_skirt_rotate_r="0,-30,-30")
   > Set Dyn_front_skirt_1_rotate(Dyn_front_skirt_1_rotate="-20,-30,0")
   > Set Dyn_front_skirt_2_rotate(Dyn_front_skirt_2_rotate="-30,-45,0")
   Set Dyn_alfa_slide(execute=@n2.else, Dyn_alfa_slide=0)
   > Set Dyn_back_slide_skirt_rotate_r(Dyn_back_slide_skirt_rotate_r="0, 0, 0")
   > Set Dyn_middle_slide_skirt_rotate_r(Dyn_middle_slide_skirt_rotate_r="0, 0, 0")
   > Set Dyn_front_skirt_1_rotate(Dyn_front_skirt_1_rotate="0, 0, 0")
   > Set Dyn_front_skirt_2_rotate(Dyn_front_skirt_2_rotate="0, 0, 0")
   n4 = Branch(execute=@n1.then_1, Condition=Get bIsCrouching())
   > Set Dyn_Dragon_head_translate(Dyn_Dragon_head_translate:Vector="-1,0,-1") -> (Output_Get:Vector)
   > Set Dyn_Dragon_head_rotate(Dyn_Dragon_head_rotate:Rotator="5,10,0") -> (Output_Get:Rotator)
   > Set Dyn_front_skirt_rotate(Dyn_front_skirt_rotate:Rotator="10,60,0") -> (Output_Get:Rotator)
   > Set Dyn_front_skirt_translate(Dyn_front_skirt_translate:Vector="-2,0,0") -> (Output_Get:Vector)
   > Set Dyn_middle_skirt_rotate_l(Dyn_middle_skirt_rotate_l:Rotator="0,90,0") -> (Output_Get:Rotator)
   > Set Dyn_middle_skirt_translate(Dyn_middle_skirt_translate:Vector="-3,0,0") -> (Output_Get:Vector)
   > Set Dyn_side_skirt_rotate_l(Dyn_side_skirt_rotate_l:Rotator="-5,0,0") -> (Output_Get:Rotator)
   > Set Dyn_back_middle_skirt_rotate_l(Dyn_back_middle_skirt_rotate_l:Rotator="-20,15,0") -> (Output_Get:Rotator)
   n5 = Branch(execute=@n4.else, Condition=Get bIsParachuteOpened())
   > Set Dyn_Dragon_head_translate(Dyn_Dragon_head_translate="-2,-6,2")
   > Set Dyn_Dragon_head_rotate(Dyn_Dragon_head_rotate="0,-30,0")
   > Set Dyn_Dragon_spine_translate(Dyn_Dragon_spine_translate:Vector="0,2,0") -> (Output_Get:Vector)
   > Set Dyn_back_skirt_rotate(Dyn_back_skirt_rotate:Rotator="0,20,0") -> (Output_Get:Rotator)
   n6 = Branch(execute=@n5.else, Condition=Get bIsSkyDiving())
   > Set Dyn_alfa_transfom(Dyn_alfa_transfom:double=1) -> (Output_Get:double)
   > Set Dyn_Dragon_head_translate(Dyn_Dragon_head_translate="-5,-4,0")
   > Set Dyn_Dragon_head_rotate(Dyn_Dragon_head_rotate="0,30,0")
   > Set Dyn_Dragon_spine_translate(Dyn_Dragon_spine_translate="0, 0, 0")
   Set Dyn_alfa_transfom(execute=@n6.else, Dyn_alfa_transfom=0)
   > Set Dyn_Dragon_head_translate(Dyn_Dragon_head_translate="0, 0, 0")
   > Set Dyn_Dragon_head_rotate(Dyn_Dragon_head_rotate="0, 0, 0")
   > Set Dyn_Dragon_spine_translate(Dyn_Dragon_spine_translate="0, 0, 0")
   > Set Dyn_front_skirt_rotate(Dyn_front_skirt_rotate="0, 0, 0")
   > Set Dyn_front_skirt_translate(Dyn_front_skirt_translate="0, 0, 0")
   > Set Dyn_middle_skirt_rotate_l(Dyn_middle_skirt_rotate_l="0, 0, 0")
   > Set Dyn_middle_skirt_translate(Dyn_middle_skirt_translate="0, 0, 0")
   > Set Dyn_side_skirt_rotate_l(Dyn_side_skirt_rotate_l="0, 0, 0")
   > Set Dyn_back_middle_skirt_rotate_l(Dyn_back_middle_skirt_rotate_l="0, 0, 0")
  ]}
]
variables [
 Dyn_alfa_slide {type=number}
 Dyn_alfa_transfom {type=number}
 Dyn_back_middle_skirt_rotate_l {type=Rotator}
 Dyn_back_skirt_rotate {type=Rotator}
 Dyn_back_slide_skirt_rotate_r {type=Rotator}
 Dyn_Dragon_head_rotate {type=Rotator}
 Dyn_Dragon_head_translate {type=Vector}
 Dyn_Dragon_spine_translate {type=Vector}
 Dyn_dragon_tranform_alfa {type=number}
 Dyn_front_skirt_1_rotate {type=Rotator}
 Dyn_front_skirt_2_rotate {type=Rotator}
 Dyn_front_skirt_rotate {type=Rotator}
 Dyn_front_skirt_translate {type=Vector}
 Dyn_middle_skirt_rotate_l {type=Rotator}
 Dyn_middle_skirt_translate {type=Vector}
 Dyn_middle_slide_skirt_rotate_r {type=Rotator}
 Dyn_side_skirt_rotate_l {type=Rotator}
]
